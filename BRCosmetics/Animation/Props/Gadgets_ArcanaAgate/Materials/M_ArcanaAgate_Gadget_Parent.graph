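# Parent material for the Arcana Agate gadget prop, adding a flipbook-driven painting layer to the shared 2023 character shading.
# MF_SubUV samples a 4x3 flipbook on the second UV channel, with parameterized scale and offset and a manually driven CurrentFrame.
# The flipbook replaces Base Color through a material-attributes blend, with a painting mask channel declared from the Tech Art Mask.
# Named reroutes carry this effect between the character PreFX and PostFX functions.
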Material /BRCosmetics/Animation/Props/Gadgets_ArcanaAgate/Materials/M_ArcanaAgate_Gadget_Parent
graph {
 properties {
  bUseMaterialAttributes=true
  bUsedWithMeshDeformer=true
  bUsedWithMorphTargets=true
  bUsedWithNiagaraMeshParticles=true
  bUsedWithSkeletalMesh=true
  pixelDepthOffsetMode=PDOM_Legacy
  shadingModel=MSM_FromMaterialExpression
  usedShadingModels=Default Lit | Subsurface | Clear Coat | Subsurface Profile | Cloth
 }
 nodes [
  n1 = MaterialFunctionCall() {materialFunction=&MF_CharacterParent_2023_PreFX}
  MaterialFunctionCall(Material=NamedRerouteUsage(), M=@n1.M, Pre Skinned Local Position=@n1.M, Distance Blend=@n1.M, Fresnel=@n1.M) {materialFunction=&MF_CharacterParent_2023_PostFX}
  NamedRerouteDeclaration(None=@n1.Result) {name=InMaterial, nodeColor=<rgba 1 0 0.375 1>}
  n2 = NamedRerouteUsage()
  n3 = TextureCoordinate() {coordinateIndex=1}
  n4 = ComponentMask(None=VectorParameter() {group=ArcanaAgata, parameterName=UVScale, defaultValue=<rgba 1.29 0.9879 0 1>}.RGB) {g=true, r=true}
  n5 = VectorParameter() {group=ArcanaAgata, parameterName=UVOffset, defaultValue=<rgba -0.15 0 0 1>}
  n6 = MaterialFunctionCall!(
   U_Images=Constant() {r=4}
   V_Images=Constant() {r=3}
   SubUV Texture RGB=TextureObject() {texture=/BRCosmetics/Animation/Props/Gadgets_ArcanaAgate/Textures/T_ArcanaAgate_Prop_Flipbook}
   Time Speed
   Manual Time Control=StaticBool() {value=true}
   Time=ScalarParameter() {group=ArcanaAgata, parameterName=CurrentFrame}
   UVs=Add(A=Multiply(A=@n3, B=@n4), B=ComponentMask(None=@n5.RGB) {g=true, r=true})
  ) {materialFunction=/Game/Effects/Fort_Effects/Materials/_Functions/MF_SubUV}
  NamedRerouteDeclaration(None=BlendMaterialAttributes(
    A=NamedRerouteUsage()
    B=SetMaterialAttributes!(MaterialAttributes=@n2, Base Color=@n6.RGB) {attributeSetTypes [#1]}
    Alpha=NamedRerouteUsage()
   )) {name=OutMaterial, nodeColor=<rgba 0.9453 1 0 1>}
  NamedRerouteDeclaration(None=ChannelMaskParameter(None=@n1.Tech Art Mask Alpha) {group=ArcanaAgata, parameterName=FXMask_Painting}.R) {name=FXMask, nodeColor=<rgba 0.7344 0 1 1>}
 ]
}
